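# HUD widget that backs the Boons inventory header, showing a title, upgrade subtitle and hero rank badge over a rank-colored background.
# Construct caches the background's dynamic material and passes the viewer ranking view model's current rank tag to OnRankChange.
# RankUpVisuals uses that tag as a row name in DT_CherrySmokeBoons_RankColors to set background and outline colors.
# UpdateRankIcon swaps the badge's soft material and hides its middle and background pips.
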WidgetBlueprint /CherrySmokeBoonsUI/HUD/WBP_CherrySmokeBoons_InventoryBacking
graphs [
 EventGraph {nodes [
   n1 = UpdateRankIcon!() -> (OutputDelegate, then, NewRankIcon:Material Instance Soft Object Reference) {is=title K2Node_CustomEvent}
   n2 = Get Image_RankIcon() -> (Image_RankIcon:Image Object Reference)
   > SetBrushFromSoftMaterial!(execute, self:Image Object Reference=@n2, SoftMaterial:Material Interface Soft Object Reference=@n1.NewRankIcon) -> (then)
   > n3 = GetDynamicMaterial(self=@n2)
   > SetScalarParameterValue(self=@n3, ParameterName=ShowMiddlePips, Value=0)
   > SetScalarParameterValue(self=@n3, ParameterName=BackgroundPipsOpacity, Value=0)
   Event Construct()
   > n4 = GetDynamicMaterial(self=Get Image_Background() -> (Image_Background:Image Object Reference))
   > Set BackgroundMID(BackgroundMID:Material Instance Dynamic Object Reference=@n4) -> (Output_Get:Material Instance Dynamic Object Reference)
   > OnRankChange!(
    execute
    self
    RankUpTag:Gameplay Tag Structure=Get CurrentRankTag!(self:Cherry Smoke Hero Ranking VM Object Reference=Get CherrySmokeViewerHeroRankingVM() -> (CherrySmokeViewerHeroRankingVM:Cherry Smoke Viewer Hero Ranking VM Object Reference)) -> (CurrentRankTag:Gameplay Tag Structure)
   ) -> (then)
   Event PreConstruct()
  ]}
 OnRankChange {nodes [
   n1 = OnRankChange!() -> (then, RankUpTag:Gameplay Tag Structure) {is=title K2Node_FunctionEntry}
   > RankUpVisuals!(execute, self, RankTagName:Gameplay Tag Structure=@n1) -> (then)
  ]}
 RankUpVisuals {nodes [
   n1 = RankUpVisuals!() -> (then, RankTagName:Gameplay Tag Structure) {is=title K2Node_FunctionEntry}
   > n2 = Get Data Table Row!(
    execute:Exec
    DataTable:Data Table Object Reference=Get RankStylingDataTable() -> (RankStylingDataTable:Data Table Object Reference)
    RowName:Name=GetTagName!(GameplayTag:"Gameplay Tag Structure (by ref)"=@n1) -> (ReturnValue:Name=None)
   ) -> (then:Exec, RowNotFound:Exec, ReturnValue:S Cherry Smoke Boons Styling Data Structure)
   n3 = Break S Cherry Smoke Boons Styling Data!(S_CherrySmokeBoons_StylingData:"S Cherry Smoke Boons Styling Data Structure (by ref)"=@n2) -> (
    RankBoxBacking_49_#1:Linear Color Structure
    RankBoxOutline_50_#2:Linear Color Structure
   ) {is=title K2Node_BreakStruct}
   > Set RankBoxBackingColor(RankBoxBackingColor:Linear Color Structure=@n3) -> (Output_Get:Linear Color Structure)
   > Set RankBoxOutlineColor(RankBoxOutlineColor:Linear Color Structure=@n3.RankBoxOutline_50_#2) -> (Output_Get:Linear Color Structure)
   > n4 = Get!(execute:Exec) -> (then:Exec, else:Exec, BackgroundMID:Material Instance Dynamic Object Reference)
   > SetVectorParameterValue(self=@n4, ParameterName=BackgroundColor, Value=Get RankBoxBackingColor() -> (RankBoxBackingColor:Linear Color Structure))
   > SetVectorParameterValue(
    self=Get BackgroundMID() -> (BackgroundMID:Material Instance Dynamic Object Reference)
    ParameterName=OutlineColor
    Value=Get RankBoxOutlineColor() -> (RankBoxOutlineColor:Linear Color Structure)
   )
  ]}
]
variables [
 BackgroundMID {type=MaterialInstanceDynamic}
 RankStylingDataTable {type=DataTable, default=&/DT_CherrySmokeBoons_RankColors}
]
widgets {
 info {inheritedWidgetCount=1, parentClass=&FortHUDElementWidget, rootWidgetClass=&SizeBox, widgetCount=10}
 widgets [
  SizeBox_0 {widgetClassPath=&SizeBox, widget=SizeBox_0, properties {bOverride_MinDesiredWidth=true, minDesiredWidth=468}}
  Overlay_12 {widgetClassPath=&Overlay, parent=SizeBox_0, widget=Overlay_12, slot {class=SizeBoxSlot, properties {padding=<ltrb 0 15 0 18>}}}
  Overlay_0 {
   widgetClassPath=&Overlay
   parent=Overlay_12
   widget=Overlay_0
   slot {class=OverlaySlot, properties {horizontalAlignment=HAlign_Fill, verticalAlignment=VAlign_Fill, padding=<ltrb -10 -5 5 -5>}}
  }
  Image_Background {
   widgetClassPath=&Image
   bIsVariable=true
   parent=Overlay_0
   widget=Image_Background
   properties {brush {
     bIsDynamicallyLoaded=false
     drawAs=Image
     imageType=FullColor
     mirroring=NoMirror
     resourceName=None
     resourceObject=/CherrySmokeBoonsUI/Materials/MI_UI_CherrySmokeBoons_BoonsInventoryBackground
     tiling=NoTile
     imageSize=<32 32>
     margin=<ltrb 0 0 0 0>
     outlineSettings {
      bUseBrushTransparency=false
      roundingType=HalfHeightRadius
      width=0
      color {colorUseRule=UseColor_Specified, specifiedColor=<rgba 0 0 0 0>}
      cornerRadii=<0 0 0 0>
     }
     tintColor {colorUseRule=UseColor_Specified, specifiedColor=<rgba 1 1 1 1>}
     uVRegion {bIsValid=false, max=<0 0>, min=<0 0>}
    }}
   slot {class=OverlaySlot, properties {horizontalAlignment=HAlign_Fill, padding=<ltrb -5 0 0 0>}}
  }
  Overlay_Text {
   widgetClassPath=&Overlay
   parent=Overlay_0
   widget=Overlay_Text
   slot {class=OverlaySlot, properties {horizontalAlignment=HAlign_Fill, verticalAlignment=VAlign_Center, padding=<ltrb 10 0 10 0>}}
  }
  HorizontalBox_0 {widgetClassPath=&HorizontalBox, parent=Overlay_Text, widget=HorizontalBox_0, slot {class=OverlaySlot}}
  CommonTextBlock_Title {
   widgetClassPath=&CommonTextBlock
   parent=HorizontalBox_0
   widget=CommonTextBlock_Title
   properties {
    mobileFontSizeMultiplier=1.45
    scrollStyle=&TextScrollStyle-Base_C
    text=Boons
    textTransformPolicy=ToUpper
    font {
     bForceMonospaced=false
     bMaterialIsStencil=false
     fontMaterial=/Game/Athena/HUD/Inventory/Art/M_UI_Inventory_CatergoryHeader_FontColor
     fontObject=&BurbankSmall
     letterSpacing=0
     monospacedWidth=1
     size=13.13
     skewAmount=0.2
     typefaceFontName=Black
     outlineSettings {
      bApplyOutlineToDropShadows=false
      bMiteredCorners=false
      bSeparateFillAlpha=true
      outlineMaterial=&M-FontShadow
      outlineSize=2
      outlineColor=<rgba 0 0 0 0.5>
     }
    }
    shadowOffset=<0 0>
   }
   slot {class=HorizontalBoxSlot}
  }
  CommonTextBlock_SubTitle {
   widgetClassPath=&CommonTextBlock
   parent=HorizontalBox_0
   widget=CommonTextBlock_SubTitle
   properties {
    mobileFontSizeMultiplier=1.4
    scrollStyle=&TextScrollStyle-Base_C
    text=All boons upgrade with hero ranking
    colorAndOpacity {colorUseRule=UseColor_Specified, specifiedColor=<rgba 1 1 1 0.9>}
    font {
     bForceMonospaced=false
     bMaterialIsStencil=false
     fontMaterial=&M_UI_Inventory_CatergoryHeader_FontColor
     fontObject=&BurbankSmall
     letterSpacing=0
     monospacedWidth=1
     size=13.13
     skewAmount=0.2
     typefaceFontName=Bold
     outlineSettings {
      bApplyOutlineToDropShadows=false
      bMiteredCorners=false
      bSeparateFillAlpha=true
      outlineMaterial=&M-FontShadow
      outlineSize=2
      outlineColor=<rgba 0 0 0 0.5>
     }
    }
    shadowOffset=<0 0>
   }
   slot {class=HorizontalBoxSlot, properties {horizontalAlignment=HAlign_Left, verticalAlignment=VAlign_Top, padding=<ltrb 12 0 0 0>}}
  }
  Image_RankIcon {
   widgetClassPath=&Image
   bIsVariable=true
   parent=Overlay_12
   widget=Image_RankIcon
   properties {
    brush {
     bIsDynamicallyLoaded=false
     drawAs=Image
     imageType=FullColor
     mirroring=NoMirror
     resourceName=None
     resourceObject=/CherrySmokeHeroRankingUI/Materials/BadgeInstanceMaterials/MI_UI_CherrySmokeHeroRank_FullBadge_C
     tiling=NoTile
     imageSize=<42 42>
     margin=<ltrb 0 0 0 0>
     outlineSettings {
      bUseBrushTransparency=false
      roundingType=HalfHeightRadius
      width=0
      color {colorUseRule=UseColor_Specified, specifiedColor=<rgba 0 0 0 0>}
      cornerRadii=<0 0 0 0>
     }
     tintColor {colorUseRule=UseColor_Specified, specifiedColor=<rgba 1 1 1 1>}
     uVRegion {bIsValid=false, max=<0 0>, min=<0 0>}
    }
    renderTransform {angle=0, scale=<1.75 1.75>, shear=<0 0>, translation=<0 0>}
   }
   slot {class=OverlaySlot, properties {horizontalAlignment=HAlign_Right, verticalAlignment=VAlign_Center, padding=<ltrb 0 -15 0 -15>}}
  }
  TutorialHighlight {widgetClassPath=&TutorialHighlight}
 ]
}
