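# Translucent UI material that draws a rounded-box background with an outline, named for the boon rank display in CherrySmokeBoons.
# An anti-aliased SDF box on 9-slice UVs provides the tinted fill, and subtracting a slightly smaller box from an outer one provides the border.
# Named reroutes carry fill and border color and alpha into a blend, with opacity scaled by an optional left-side gradient fade and MasterOpacity.
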Material /CherrySmokeBoonsUI/Materials/M_UI_CherrySmokeBoons_BoonRankBackground
graph {
 properties {bAutomaticallySetUsageInEditor=false, blendMode=BLEND_Translucent, materialDomain=MD_UI, usedShadingModels=Default Lit}
 nodes [
  n1 = Max(B=NamedRerouteUsage())
  LinearInterpolate(A=NamedRerouteUsage(), B=NamedRerouteUsage(), Alpha=@n1)
  n2 = Multiply(A=ComponentMask(None=MaterialFunctionCall!(UVs) {materialFunction=/FortUILibrary/MaterialFunctions/Basics/MF_UI_9SliceUVs}.UVs) {r=true}) {constB=2}
  n3 = Saturate(None=Subtract(A=Saturate(None=@n2)) {constB=0.2})
  n4 = ScalarParameter() {defaultValue=1, parameterName=FadeLeftSide}
  Multiply(A=Multiply(A=@n1, B=LinearInterpolate(B=@n3, Alpha=@n4) {constA=1}), B=ScalarParameter() {defaultValue=1, parameterName=MasterOpacity})
  n5 = MaterialFunctionCall() {materialFunction=&MF_UI_9SliceUVs}
  n6 = ScalarParameter() {defaultValue=0.8, parameterName=BoxSize}
  n7 = ScalarParameter() {defaultValue=0.1, parameterName=BorderThickness}
  n8 = MaterialFunctionCall(DistanceField=MaterialFunctionCall!(
    UVs=Clamp(None=@n5.UVs)
    Size=Add(A=@n6, B=@n7)
    CornerRoundness=ScalarParameter() {defaultValue=0.9, parameterName=BoxRoundness}
    KeepAspectRatio
    PixelSize
   ) {materialFunction=/FortUILibrary/MaterialFunctions/Shapes/SDFs/MF_SDF_Box}.SDF) {materialFunction=&MF_UI_AntiAliasing}
  NamedRerouteDeclaration(None=Multiply(A=@n8.Inner, B=VectorParameter() {parameterName=BackgroundColor, defaultValue=<rgba 0.0926 0.5905 1 1>}.RGB)) {name=BackgroundColor, nodeColor=<rgba 0 0 0 1>}
  NamedRerouteDeclaration(None=Multiply(A=@n8.Inner, B=ScalarParameter() {defaultValue=0.5, parameterName=BackgroundOpacity})) {name=AA'dBackground, nodeColor=<rgba 0 0 0 1>}
  n9 = Clamp(None=MaterialFunctionCall() {materialFunction=&MF_UI_9SliceUVs}.UVs)
  n10 = Add(A=ScalarParameter() {defaultValue=0.8, parameterName=BoxSize}, B=ScalarParameter() {defaultValue=0.1, parameterName=BorderThickness})
  n11 = ScalarParameter() {defaultValue=0.9, parameterName=BoxRoundness}
  n12 = Clamp(None=MaterialFunctionCall() {materialFunction=&MF_UI_9SliceUVs}.UVs)
  n13 = ScalarParameter() {defaultValue=0.8, parameterName=BoxSize}
  n14 = Add(A=ScalarParameter() {defaultValue=0.9, parameterName=BoxRoundness}) {constB=-0.1}
  n15 = MaterialFunctionCall(DistanceField=MaterialFunctionCall!(
    DistanceFieldA=MaterialFunctionCall(UVs=@n9, Size=@n10, CornerRoundness=@n11) {materialFunction=&MF_SDF_Box}.SDF
    DistanceFieldB=MaterialFunctionCall(UVs=@n12, Size=@n13, CornerRoundness=@n14) {materialFunction=&MF_SDF_Box}.SDF
    BlendAmount
   ) {materialFunction=/FortUILibrary/MaterialFunctions/Shapes/Modifiers/MF_UI_SDF_Subtraction}.Exact) {materialFunction=&MF_UI_AntiAliasing}
  NamedRerouteDeclaration(None=Multiply(A=VectorParameter() {parameterName=OutlineColor, defaultValue=<rgba 0.2153 0.5925 1 1>}.RGB, B=@n15.Inner)) {name=BorderColor, nodeColor=<rgba 0 0 0 1>}
  NamedRerouteDeclaration(None=@n15.Inner) {name=AA'dBorder, nodeColor=<rgba 0 0 0 1>}
  NamedRerouteUsage()
 ]
}
