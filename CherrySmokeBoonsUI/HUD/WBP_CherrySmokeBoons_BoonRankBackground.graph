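# HUD widget for CherrySmoke boons: a hero rank icon beside a boons container on a rank-colored backing.
# Binds to the view target player's augment component, hiding when the container is empty and fading in at one child.
# Rank changes read colors from a tag-keyed data table into the background's dynamic material and play a rank-up animation.
# UpdateRankIcon swaps the badge material; PreConstruct supports designer rank previews.
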WidgetBlueprint /CherrySmokeBoonsUI/HUD/WBP_CherrySmokeBoons_BoonRankBackground
animations [
 Anim_Show {
  binding_count=1
  length_seconds=0.25
  object=&self:Anim_Show
  sequence {bindings [Overlay_BoonBacking {
     id=#1
     possessed_class=Overlay
     tracks [{
       type=MovieSceneFloatTrack
       display_name=Render Opacity
       sections [{type=MovieSceneFloatSection, end_frame=5, start_frame=0, channels [None {type=MovieSceneScriptingFloatChannel, keys [[0,0],[5,1]]}]}]
      }]
    }]}
 }
 Anim_RankUp {
  binding_count=2
  length_seconds=0.5
  object=&self:Anim_RankUp
  sequence {bindings [
    Image_RankIcon {
     id=#2
     possessed_class=Image
     tracks [{type=MovieSceneWidgetMaterialTrack, sections [{type=MovieSceneParameterSection, end_frame=10, start_frame=0}]}]
    }
    Overlay_RankIcon {
     id=#3
     possessed_class=Overlay
     tracks [{
       type=MovieScene2DTransformTrack
       sections [{
         type=MovieScene2DTransformSection
         end_frame=10
         start_frame=0
         channels [
          Translation.X {type=MovieSceneScriptingFloatChannel, const [0,1]}
          Translation.Y {type=MovieSceneScriptingFloatChannel, const [0,1]}
          Angle {type=MovieSceneScriptingFloatChannel, const [0,1]}
          Scale.X {type=MovieSceneScriptingFloatChannel, keys [[0,1],[3,0.85],[6,0.85],[8,1.2],[10,1]]}
          Scale.Y {type=MovieSceneScriptingFloatChannel, keys [[0,1],[3,0.85],[6,0.85],[8,1.2],[10,1]]}
          Shear.X {type=MovieSceneScriptingFloatChannel, const [0,1]}
          Shear.Y {type=MovieSceneScriptingFloatChannel, const [0,1]}
         ]
        }]
      }]
    }
   ]}
 }
]
graphs [
 EventGraph {nodes [
   n1 = OnViewTargetChanged!() -> (
    OutputDelegate
    then
    PlayerController:Fort Player Controller Object Reference
    OldViewTarget:Actor Object Reference
    NewViewTarget:Actor Object Reference
   ) {is=title K2Node_CustomEvent}
   > Is Valid(InputObject=@n1.NewViewTarget)
   > n2 = Cast To FortPlayerPawn(Object=@n1.NewViewTarget)
   n3 = Get PlayerState!(self:Pawn Object Reference=@n2) -> (PlayerState:Player State Object Reference)
   > n4 = Cast To FortPlayerStateAthena!(execute, Object=@n3) -> (then, CastFailed, AsFort Player State Athena:Fort Player State Athena Object Reference) {is=title K2Node_DynamicCast}
   > Set Player State Ref(Player State Ref:Fort Player State Athena Object Reference=@n4) -> (Output_Get:Fort Player State Athena Object Reference)
   > Bind Event to On Augment List Updated(
    self:Fort Player State Component Player Augment System Object Reference=GetAugmentComponent!(PlayerContextActor:Actor Object Reference=@n3) -> (ReturnValue:Fort Player State Component Player Augment System Object Reference)
    Delegate=Create Event()
   ) {is=title K2Node_AddDelegate}
   > OnAugmentListUpdated!(execute, self, AugmentComponent:Fort Player State Component Player Augment System Object Reference) -> (then)
   n5 = UpdateRankIcon!() -> (OutputDelegate, then, NewRankIcon:Material Instance Soft Object Reference) {is=title K2Node_CustomEvent}
   n6 = Get Image_RankIcon() -> (Image_RankIcon:Image Object Reference)
   > SetBrushFromSoftMaterial!(execute, self:Image Object Reference=@n6, SoftMaterial:Material Interface Soft Object Reference=@n5.NewRankIcon) -> (then)
   > n7 = GetDynamicMaterial(self=@n6)
   > SetScalarParameterValue(self=@n7, ParameterName=ShowMiddlePips, Value=0)
   > SetScalarParameterValue(self=@n7, ParameterName=BackgroundPipsOpacity, Value=0)
   OnAugmentListUpdated!() -> (OutputDelegate, then, AugmentComponent:Fort Player State Component Player Augment System Object Reference) {is=title K2Node_CustomEvent}
   > DelayUntilNextTick!(execute) -> (then)
   n8 = Get HorizontalBox_BoonsContainer() -> (HorizontalBox_BoonsContainer:Horizontal Box Object Reference)
   > n9 = Branch(Condition="integer > integer"(A=GetChildrenCount!(self:Panel Widget Object Reference=@n8) -> (ReturnValue:int=0)))
   > SetVisibility(InVisibility=HitTestInvisible)
   n10 = Get HorizontalBox_BoonsContainer()
   > Branch(Condition="Equal (Integer)"(A=GetChildrenCount(self=@n10) -> (ReturnValue=0), B=1))
   > PlayAnimationForward(InAnimation=Get Anim_Show() -> (Anim_Show:Widget Animation Object Reference))
   SetVisibility(execute=@n9.else, InVisibility=Collapsed)
   Event Construct()
   > n11 = Cast To Athena_PlayerController!(execute, Object=GetOwningPlayer()) -> (then, CastFailed, AsAthena Player Controller:Athena Player Controller Object Reference) {is=title K2Node_DynamicCast}
   > Bind Event to On View Target Changed(self:Fort Player Controller Object Reference=@n11, Delegate=Create Event()) {is=title K2Node_AddDelegate}
   > OnViewTargetChanged!(
    execute
    self
    PlayerController:Fort Player Controller Object Reference
    OldViewTarget:Actor Object Reference
    NewViewTarget:Actor Object Reference=Get Controlled Pawn!(self:Controller Object Reference=@n11) -> (ReturnValue:Pawn Object Reference) {is=title K2Node_CallFunction}
   ) -> (then) {is=title K2Node_CallFunction}
   > n12 = GetDynamicMaterial(self:Border Object Reference=Get Border_RankBackground() -> (Border_RankBackground:Border Object Reference))
   > Set BorderRankBackgroundMID(BorderRankBackgroundMID:Material Instance Dynamic Object Reference=@n12) -> (Output_Get:Material Instance Dynamic Object Reference)
   > OnRankChange!(
    execute
    self
    RankUpTag:Gameplay Tag Structure=Get CurrentRankTag!(self:Cherry Smoke Hero Ranking VM Object Reference=Get CherrySmokeViewerHeroRankingVM() -> (CherrySmokeViewerHeroRankingVM:Cherry Smoke Viewer Hero Ranking VM Object Reference)) -> (CurrentRankTag:Gameplay Tag Structure)
   ) -> (then)
   n13 = Event PreConstruct()
   > Branch(Condition=@n13.IsDesignTime)
   > n14 = Branch(Condition="Not Equal (Enum)"!(A:BoonsRankEnum Enum=Get Preview Rank Enum() -> (Preview Rank Enum:BoonsRankEnum Enum=NewEnumerator5), B:BoonsRankEnum Enum=NewEnumerator5) -> (ReturnValue:bool))
   > n15 = Get!(execute:Exec) -> (then:Exec, else:Exec, BorderRankBackgroundMID:Material Instance Dynamic Object Reference)
   OnRankChange(
    execute=[@n17.then, @n15.then]
    RankUpTag=Select!(
     NewEnumerator5:Gameplay Tag Structure
     NewEnumerator0:Gameplay Tag Structure=Get CRankTag() -> (CRankTag:Gameplay Tag Structure)
     NewEnumerator1:Gameplay Tag Structure=Get BRankTag() -> (BRankTag:Gameplay Tag Structure)
     NewEnumerator2:Gameplay Tag Structure=Get ARankTag() -> (ARankTag:Gameplay Tag Structure)
     NewEnumerator3:Gameplay Tag Structure=Get SRankTag() -> (SRankTag:Gameplay Tag Structure)
     NewEnumerator4:Gameplay Tag Structure=Get SPlusRankTag() -> (SPlusRankTag:Gameplay Tag Structure)
     Index:BoonsRankEnum Enum=Get Preview Rank Enum() -> (Preview Rank Enum=NewEnumerator5)
    ) -> (ReturnValue:Gameplay Tag Structure)
   )
   n16 = GetDynamicMaterial(execute=@n15.else, self=Get Border_RankBackground())
   > n17 = Set BorderRankBackgroundMID(BorderRankBackgroundMID=@n16)
   SetBrushFromMaterial(
    execute=@n14.else
    self:Border Object Reference=Get Border_RankBackground()
    Material=Get DefaultRankBackgroundMaterial() -> (DefaultRankBackgroundMaterial:Material Interface Object Reference)
   )
   n18 = Get Image_RankIcon()
   > SetBrushFromSoftMaterial(self=@n18, SoftMaterial=Get DefaultRankIconMaterial() -> (DefaultRankIconMaterial:Material Instance Soft Object Reference))
   > n19 = GetDynamicMaterial(self:Image Object Reference=@n18)
   > SetScalarParameterValue(self=@n19, ParameterName=ShowMiddlePips, Value=0)
   > SetScalarParameterValue(self=@n19, ParameterName=BackgroundPipsOpacity, Value=0)
  ]}
 OnRankChange {nodes [
   n1 = OnRankChange!() -> (then, RankUpTag:Gameplay Tag Structure) {is=title K2Node_FunctionEntry}
   > RankUpVisuals!(execute, self, RankTagName:Gameplay Tag Structure=@n1) -> (then)
   > Branch(Condition="Not Equal (Enum)"(A:ESlateVisibility Enum=Get Visibility() -> (Visibility:ESlateVisibility Enum=Visible), B:ESlateVisibility Enum=Collapsed))
   > PlayAnimationForward(InAnimation=Get Anim_RankUp() -> (Anim_RankUp:Widget Animation Object Reference))
  ]}
 RankUpVisuals {nodes [
   n1 = RankUpVisuals!() -> (then, RankTagName:Gameplay Tag Structure) {is=title K2Node_FunctionEntry}
   > n2 = Get Data Table Row!(
    execute:Exec
    DataTable:Data Table Object Reference=Get RankStylingDataTable() -> (RankStylingDataTable:Data Table Object Reference)
    RowName:Name=GetTagName!(GameplayTag:"Gameplay Tag Structure (by ref)"=@n1) -> (ReturnValue:Name=None)
   ) -> (then:Exec, RowNotFound:Exec, ReturnValue:S Cherry Smoke Boons Styling Data Structure)
   n3 = Break S Cherry Smoke Boons Styling Data!(S_CherrySmokeBoons_StylingData:"S Cherry Smoke Boons Styling Data Structure (by ref)"=@n2) -> (
    RankBoxBacking_49_#4:Linear Color Structure
    RankBoxOutline_50_#5:Linear Color Structure
   ) {is=title K2Node_BreakStruct}
   > Set RankBoxBackingColor(RankBoxBackingColor:Linear Color Structure=@n3) -> (Output_Get:Linear Color Structure)
   > Set RankBoxOutlineColor(RankBoxOutlineColor:Linear Color Structure=@n3.RankBoxOutline_50_#5) -> (Output_Get:Linear Color Structure)
   > n4 = Get()
   > SetVectorParameterValue(self=@n4, ParameterName=BackgroundColor, Value=Get RankBoxBackingColor() -> (RankBoxBackingColor:Linear Color Structure))
   > SetVectorParameterValue(
    self=Get BorderRankBackgroundMID() -> (BorderRankBackgroundMID:Material Instance Dynamic Object Reference)
    ParameterName=OutlineColor
    Value=Get RankBoxOutlineColor() -> (RankBoxOutlineColor:Linear Color Structure)
   )
  ]}
]
variables [
 ARankTag {type=GameplayTag, default {tagName=CherrySmoke.HeroRank.A}}
 BorderRankBackgroundMID {type=MaterialInstanceDynamic}
 BRankTag {type=GameplayTag, default {tagName=CherrySmoke.HeroRank.B}}
 CRankTag {type=GameplayTag, default {tagName=CherrySmoke.HeroRank.C}}
 DefaultRankBackgroundMaterial {type=MaterialInterface, default=/CherrySmokeBoonsUI/Materials/M_UI_CherrySmokeBoons_BoonRankBackground}
 DefaultRankIconMaterial {type=MaterialInstance, default=/CherrySmokeHeroRankingUI/Materials/BadgeInstanceMaterials/MI_UI_CherrySmokeHeroRank_FullBadge_C}
 Player State Ref {type=FortPlayerStateAthena, default=null}
 Preview Rank Enum {type=BoonsRankEnum, default=Default}
 RankStylingDataTable {type=DataTable, default=&/DT_CherrySmokeBoons_RankColors}
 SPlusRankTag {type=GameplayTag, default {tagName=CherrySmoke.HeroRank.S+}}
 SRankTag {type=GameplayTag, default {tagName=CherrySmoke.HeroRank.S}}
]
widgets {
 info {inheritedWidgetCount=1, parentClass=&FortHUDElementWidget, rootWidgetClass=&Overlay, widgetCount=7}
 widgets [
  Overlay_BoonBacking {widgetClassPath=&Overlay, widget=Overlay_BoonBacking, properties {bOverride_Cursor=true, visibility=Visible}}
  Border_RankBackground {
   widgetClassPath=&Border
   bIsVariable=true
   parent=Overlay_BoonBacking
   widget=Border_RankBackground
   properties {
    horizontalAlignment=HAlign_Left
    verticalAlignment=VAlign_Top
    background {
     bIsDynamicallyLoaded=false
     drawAs=Image
     imageType=FullColor
     mirroring=NoMirror
     resourceName=None
     resourceObject=&M_UI_CherrySmokeBoons_BoonRankBackground
     tiling=NoTile
     imageSize=<32 32>
     margin=<ltrb 0 0 0 0>
     outlineSettings {
      bUseBrushTransparency=false
      roundingType=HalfHeightRadius
      width=0
      color {colorUseRule=UseColor_Specified, specifiedColor=<rgba 0 0 0 0>}
      cornerRadii=<0 0 0 0>
     }
     tintColor {colorUseRule=UseColor_Specified, specifiedColor=<rgba 1 1 1 1>}
     uVRegion {bIsValid=false, max=<0 0>, min=<0 0>}
    }
    desiredSizeScale=<1 1.1>
    padding=<ltrb 0 2 0 0>
   }
   slot {class=OverlaySlot, properties {horizontalAlignment=HAlign_Fill, verticalAlignment=VAlign_Fill, padding=<ltrb 0 5 0 5>}}
  }
  HorizontalBox_0 {
   widgetClassPath=&HorizontalBox
   parent=Border_RankBackground
   widget=HorizontalBox_0
   properties {bOverride_Cursor=true}
   slot {class=BorderSlot, properties {horizontalAlignment=HAlign_Left, verticalAlignment=VAlign_Top, padding=<ltrb 0 2 0 0>}}
  }
  Overlay_RankIcon {widgetClassPath=&Overlay, parent=HorizontalBox_0, widget=Overlay_RankIcon, slot {class=HorizontalBoxSlot}}
  Image_RankIcon {
   widgetClassPath=&Image
   bIsVariable=true
   parent=Overlay_RankIcon
   widget=Image_RankIcon
   properties {
    brush {
     bIsDynamicallyLoaded=false
     drawAs=Image
     imageType=FullColor
     mirroring=NoMirror
     resourceName=None
     resourceObject=null
     tiling=NoTile
     imageSize=<38 38>
     margin=<ltrb 0 0 0 0>
     outlineSettings {
      bUseBrushTransparency=false
      roundingType=HalfHeightRadius
      width=0
      color {colorUseRule=UseColor_Specified, specifiedColor=<rgba 0 0 0 0>}
      cornerRadii=<0 0 0 0>
     }
     tintColor {colorUseRule=UseColor_Specified, specifiedColor=<rgba 1 1 1 1>}
     uVRegion {bIsValid=false, max=<0 0>, min=<0 0>}
    }
    renderTransform {angle=0, scale=<1.5 1.5>, shear=<0 0>, translation=<0 0>}
   }
   slot {class=OverlaySlot, properties {verticalAlignment=VAlign_Center, padding=<ltrb -5 -8 0 -8>}}
  }
  HorizontalBox_BoonsContainer {
   widgetClassPath=&HorizontalBox
   bIsVariable=true
   parent=HorizontalBox_0
   widget=HorizontalBox_BoonsContainer
   slot {class=HorizontalBoxSlot, properties {padding=<ltrb 5 0 10 0>}}
  }
  TutorialHighlight {widgetClassPath=&TutorialHighlight}
 ]
}
