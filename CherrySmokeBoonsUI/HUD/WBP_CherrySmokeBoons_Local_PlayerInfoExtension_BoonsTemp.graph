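# HUD player-info extension widget that shows boss and minor medallion augment icons for the pawn being viewed.
# View-target changes rebuild the rows and bind to the player state's augment component list updates.
# Active augments are split by gameplay tag, with minor icons placed inside the boon rank background's container.
# Updates create only missing icons, then prune stale ones and rebuild the populated maps; touch toggles inventory.
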WidgetBlueprint /CherrySmokeBoonsUI/HUD/WBP_CherrySmokeBoons_Local_PlayerInfoExtension_BoonsTemp
graphs [
 AddNewBossMedallionWidgets {nodes [
   AddNewBossMedallionWidgets() {is=title K2Node_FunctionEntry}
   > n1 = For Each Loop(Array:Array of Fort Player Augment Item Definition Object References=Get NewBossMedallions() -> (NewBossMedallions:Array of Fort Player Augment Item Definition Object References)) -> (Array Element:Fort Player Augment Item Definition Object Reference)
   > n2 = Branch(Condition=Contains!(
     TargetMap:Map of Fort Player Augment Item Definition Object References to WBP Medallion Icon Object References=Get Temp_PopulatedBossMedallions() -> (Temp_PopulatedBossMedallions:Map of Fort Player Augment Item Definition Object References to WBP Medallion Icon Object References)
     Key:"Fort Player Augment Item Definition Object Reference (by ref)"=@n1
    ) -> (ReturnValue))
   n3 = Create Widget!(
    execute:Exec=@n2.else
    Class:User Widget Class Reference=/MedallionUI/HUD/WBP_Medallion_Icon.WBP_Medallion_Icon_C
    OwningPlayer:Player Controller Object Reference
   ) -> (then:Exec, ReturnValue:WBP Medallion Icon Object Reference)
   > Set BossMedallionWidget(BossMedallionWidget:WBP Medallion Icon Object Reference=@n3) -> (Output_Get:WBP Medallion Icon Object Reference)
   > n4 = AddChildToHorizontalBox!(
    execute
    self:Horizontal Box Object Reference=Get HorizontalBox_BossMedallionIcons() -> (HorizontalBox_BossMedallionIcons:Horizontal Box Object Reference)
    Content:Widget Object Reference=@n3
   ) -> (then, ReturnValue:Horizontal Box Slot Object Reference)
   > Set Padding!(
    execute
    Padding:Margin Structure=Get BossMedallionPadding() -> (BossMedallionPadding:Margin Structure)
    self:Horizontal Box Slot Object Reference=@n4
   ) -> (then, Output_Get:Margin Structure)
   > Event SetInfo!(execute, self:WBP Medallion Icon Object Reference=@n3, InItemDef:Fort Player Augment Item Definition Object Reference=@n1) -> (then)
   > SetRenderScale!(
    execute
    self:Widget Object Reference=Get BossMedallionWidget() -> (BossMedallionWidget:WBP Medallion Icon Object Reference)
    Scale:Vector 2D Structure=To Vector2D!(InDouble:double=Get BossMedallionSize() -> (BossMedallionSize:float)) -> (ReturnValue:Vector 2D Structure)
   ) -> (then)
   Return Node(execute=@n1.Completed)
  ]}
 AddNewMinorMedallionWidgets {nodes [
   AddNewMinorMedallionWidgets() {is=title K2Node_FunctionEntry}
   > n1 = For Each Loop(Array=Get NewMinorMedallions() -> (NewMinorMedallions:Array of Fort Player Augment Item Definition Object References))
   > n2 = Branch(Condition=Contains(
     TargetMap=Get Temp_PopulatedMinorMedallions() -> (Temp_PopulatedMinorMedallions:Map of Fort Player Augment Item Definition Object References to WBP Medallion Icon Object References)
     Key=@n1
    ))
   n3 = Create Widget(execute=@n2.else, Class=&WBP_Medallion_Icon_C)
   > Set WBP_MinorMedallionIcon(WBP_MinorMedallionIcon:WBP Medallion Icon Object Reference=@n3) -> (Output_Get:WBP Medallion Icon Object Reference)
   > n4 = AddChildToHorizontalBox(self=Get TEMP_HorizontalBoxBoonsContainer() -> (TEMP_HorizontalBoxBoonsContainer:Horizontal Box Object Reference), Content=@n3)
   > Set Padding(Padding=Get MinorMedallionPadding() -> (MinorMedallionPadding:Margin Structure), self=@n4)
   > Event SetInfo(self=Get WBP_MinorMedallionIcon() -> (WBP_MinorMedallionIcon:WBP Medallion Icon Object Reference), InItemDef=@n1)
   > SetRenderScale(self=Get WBP_MinorMedallionIcon(), Scale=To Vector2D(InDouble=Get MinorMedallionSize() -> (MinorMedallionSize:float)))
   Return Node(execute=@n1.Completed)
   Get HorizontalBox_MinorMedallionIcons() -> (HorizontalBox_MinorMedallionIcons:Horizontal Box Object Reference)
  ]}
 Clear Widgets {nodes [
   Clear Widgets() {is=title K2Node_FunctionEntry}
   > ClearChildren(self=[Get TEMP_HorizontalBoxBoonsContainer(), Get HorizontalBox_MinorMedallionIcons(), Get HorizontalBox_BossMedallionIcons()])
   > Clear(TargetArray:Array of Fort Player Augment Item Definition Object References=Get NewBossMedallions())
   > Clear!(
    execute
    TargetMap:Map of Fort Player Augment Item Definition Object References to WBP Medallion Icon Object References=Get Temp_PopulatedBossMedallions()
   ) -> (then) {is=title K2Node_CallFunction}
   > Clear(TargetArray=Get NewMinorMedallions()) {is=title K2Node_CallArrayFunction}
   > Clear(TargetMap=Get Temp_PopulatedMinorMedallions()) {is=title K2Node_CallFunction}
  ]}
 EventGraph {nodes [
   n1 = OnViewTargetChanged!() -> (
    OutputDelegate
    then
    PlayerController:Fort Player Controller Object Reference
    OldViewTarget:Actor Object Reference
    NewViewTarget:Actor Object Reference
   ) {is=title K2Node_CustomEvent}
   > n2 = Is Valid(InputObject=@n1.NewViewTarget)
   > n3 = Cast To FortPlayerPawn(Object=@n1.NewViewTarget)
   > Clear Widgets() {is=title K2Node_CallFunction}
   > n4 = Set CurrentPawn(CurrentPawn:Fort Player Pawn Object Reference=@n3) -> (Output_Get:Fort Player Pawn Object Reference)
   > n5 = Cast To FortPlayerStateAthena!(execute, Object=Get PlayerState!(self:Pawn Object Reference=@n4) -> (PlayerState:Player State Object Reference)) -> (then, CastFailed, AsFort Player State Athena:Fort Player State Athena Object Reference) {is=title K2Node_DynamicCast}
   > Set Player State Ref(Player State Ref:Fort Player State Athena Object Reference=@n5) -> (Output_Get:Fort Player State Athena Object Reference)
   n6 = Get TEMP_HorizontalBoxBoonsContainer()
   n7 = Get HorizontalBox_BossMedallionIcons()
   n8 = Get HorizontalBox_MinorMedallionIcons()
   > SetVisibility(self=[@n6, @n7, @n8], InVisibility=HitTestInvisible)
   > Bind Event to On Augment List Updated(
    self:Fort Player State Component Player Augment System Object Reference=GetAugmentComponent!(PlayerContextActor:Actor Object Reference=Get Player State Ref() -> (Player State Ref:Fort Player State Athena Object Reference)) -> (ReturnValue:Fort Player State Component Player Augment System Object Reference)
    Delegate=Create Event()
   ) {is=title K2Node_AddDelegate}
   > UpdateActiveMedallions!(
    execute
    self
    AugmentComponent:Fort Player State Component Player Augment System Object Reference=GetAugmentComponent(PlayerContextActor=Get Player State Ref())
   ) -> (then)
   SetVisibility(execute=@n2.Is Not Valid, self=[@n6, @n7, @n8], InVisibility=Collapsed)
   Event Construct()
   > Parent: Construct()
   > Set TEMP_HorizontalBoxBoonsContainer(TEMP_HorizontalBoxBoonsContainer:Horizontal Box Object Reference=Get HorizontalBox_BoonsContainer!(self:WBP Cherry Smoke Boons Boon Rank Background Object Reference=Get WBP_CherrySmokeBoons_BoonRankBackground() -> (WBP_CherrySmokeBoons_BoonRankBackground:WBP Cherry Smoke Boons Boon Rank Background Object Reference)) -> (HorizontalBox_BoonsContainer:Horizontal Box Object Reference)) -> (Output_Get:Horizontal Box Object Reference)
   > n9 = Cast To Athena_PlayerController!(execute, Object=GetOwningPlayer()) -> (then, CastFailed, AsAthena Player Controller:Athena Player Controller Object Reference) {is=title K2Node_DynamicCast}
   > Bind Event to On View Target Changed(self:Fort Player Controller Object Reference=@n9, Delegate=Create Event()) {is=title K2Node_AddDelegate}
   n10 = Get Controlled Pawn!(self:Controller Object Reference=@n9) -> (ReturnValue:Pawn Object Reference) {is=title K2Node_CallFunction}
   > OnViewTargetChanged!(
    execute
    self
    PlayerController:Fort Player Controller Object Reference
    OldViewTarget:Actor Object Reference
    NewViewTarget:Actor Object Reference=@n10
   ) -> (then) {is=title K2Node_CallFunction}
   > Clear Widgets()
   > n11 = Cast To FortPlayerPawn(Object=@n10)
   > n12 = Set CurrentPawn(CurrentPawn=@n11)
   > UpdateActiveMedallions(AugmentComponent=GetAugmentComponent(PlayerContextActor=Get PlayerState(self=@n12)))
   n13 = Event PreConstruct()
   > Branch(Condition=@n13.IsDesignTime)
   > Set TEMP_HorizontalBoxBoonsContainer(TEMP_HorizontalBoxBoonsContainer=Get HorizontalBox_BoonsContainer(self=Get WBP_CherrySmokeBoons_BoonRankBackground()))
   > Clear Widgets()
   > n14 = For Loop(LastIndex=3)
   n15 = Create Widget(execute=@n14.Completed, Class=&WBP_Medallion_Icon_C)
   > SetRenderScale(self=@n15, Scale=To Vector2D(InDouble=Get MinorMedallionSize()))
   > n16 = AddChildToHorizontalBox(self=Get HorizontalBox_BoonsContainer(self=Get WBP_CherrySmokeBoons_BoonRankBackground()), Content=@n15)
   > Set Padding(Padding=Get MinorMedallionPadding(), self=@n16)
   n17 = Create Widget(execute=@n14.LoopBody, Class=&WBP_Medallion_Icon_C)
   > SetRenderScale(self=@n17, Scale=To Vector2D(InDouble=Get BossMedallionSize()))
   > n18 = AddChildToHorizontalBox(self=Get HorizontalBox_BossMedallionIcons(), Content=@n17)
   > Set Padding(Padding=Get BossMedallionPadding(), self=@n18)
   Event Destruct()
   > Parent: Destruct()
   > n19 = Cast To Athena_PlayerController(Object=GetOwningPlayer())
   > Assign On View Target Changed!(execute:Exec, self:Fort Player Controller Object Reference=@n19, Delegate:"Delegate (by ref)"=Create Event()) -> (then:Exec) {is=title K2Node_AssignDelegate}
   Get HorizontalBox_MinorMedallionIcons()
  ]}
 GetOwnedGameplayTagBugWorkaround {nodes [
   n1 = GetOwnedGameplayTagBugWorkaround!() -> (then, GameplayTagAssetInterface:Gameplay Tag Asset Interface Interface) {is=title K2Node_FunctionEntry}
   > Return Node!(
    execute
    TagContainer:Gameplay Tag Container Structure=Get Owned Gameplay Tags!(self:Gameplay Tag Asset Interface Interface=@n1) -> (ReturnValue:Gameplay Tag Container Structure) {is=title K2Node_CallFunction}
   )
  ]}
 OnTouchStarted {nodes [
   OnTouchStarted!() -> (then, MyGeometry:Geometry Structure, InTouchEvent:"Pointer Event Structure (by ref)") {is=title K2Node_FunctionEntry}
   > n1 = Cast To FortPlayerController(Object=GetOwningPlayer())
   > ToggleInventory(self:Fort Player Controller Object Reference=@n1)
   > Return Node!(execute, ReturnValue:Event Reply Structure=Handled())
   Return Node(execute=@n1.CastFailed, ReturnValue=Unhandled())
  ]}
 RefreshPopulatedWidgets {nodes [
   RefreshPopulatedWidgets() {is=title K2Node_FunctionEntry}
   > Clear(TargetMap=Get Temp_PopulatedBossMedallions())
   > Clear(TargetMap=Get Temp_PopulatedMinorMedallions())
   > n1 = Set SavedChildren(SavedChildren:Array of Widget Object References=GetAllChildren(self=Get HorizontalBox_BossMedallionIcons())) -> (Output_Get:Array of Widget Object References)
   > n2 = For Each Loop(Array:Array of Widget Object References=@n1) -> (Array Element:Widget Object Reference)
   > n3 = Cast To WBP_Medallion_Icon!(execute, Object=@n2) -> (then, CastFailed, AsWBP Medallion Icon:WBP Medallion Icon Object Reference) {is=title K2Node_DynamicCast}
   n4 = GetItemDef!(self:WBP Medallion Icon Object Reference=@n3) -> (MedallionItemDef:Fort Player Augment Item Definition Object Reference)
   > n5 = Branch(Condition=Contains Item(
     TargetArray:Array of Fort Player Augment Item Definition Object References=Get NewBossMedallions()
     ItemToFind:"Fort Player Augment Item Definition Object Reference (by ref)"=@n4
    ))
   > Add!(
    execute
    TargetMap:Map of Fort Player Augment Item Definition Object References to WBP Medallion Icon Object References=Get Temp_PopulatedBossMedallions()
    Key:"Fort Player Augment Item Definition Object Reference (by ref)"=@n4
    Value:"WBP Medallion Icon Object Reference (by ref)"=@n3
   ) -> (then) {is=title K2Node_CallFunction}
   Event Remove(execute=@n5.else, self:WBP Medallion Icon Object Reference=@n3)
   > RemoveChild(self=Get HorizontalBox_BossMedallionIcons(), Content=@n2)
   n6 = Set SavedChildren(execute=@n2.Completed, SavedChildren=GetAllChildren(self=Get TEMP_HorizontalBoxBoonsContainer()))
   > n7 = For Each Loop(Array=@n6)
   > n8 = Cast To WBP_Medallion_Icon(Object=@n7)
   n9 = GetItemDef(self=@n8)
   > n10 = Branch(Condition=Contains Item(TargetArray=Get NewMinorMedallions(), ItemToFind=@n9))
   > Add(TargetMap=Get Temp_PopulatedMinorMedallions(), Key=@n9, Value=@n8)
   Event Remove(execute=@n10.else, self=@n8)
   > RemoveChild(self=Get TEMP_HorizontalBoxBoonsContainer(), Content=@n7)
   > Return Node!(execute)
   Get HorizontalBox_MinorMedallionIcons()
   Get HorizontalBox_MinorMedallionIcons()
  ]}
 SetMedallionSize {nodes [
   n1 = SetMedallionSize!() -> (then, SizeBox:Size Box Object Reference, Size:float=~) {is=title K2Node_FunctionEntry}
   > SetHeightOverride(self=@n1, InHeightOverride=@n1.Size)
   > SetWidthOverride(self=Get SizeBox() -> (SizeBox:Size Box Object Reference), InWidthOverride=@n1.Size)
  ]}
 UpdateActiveMedallions {nodes [
   UpdateActiveMedallions!() -> (then, AugmentComponent:Fort Player State Component Player Augment System Object Reference) {is=title K2Node_FunctionEntry}
   > Clear(TargetArray=Get NewBossMedallions()) {is=title K2Node_CallArrayFunction}
   > Clear(TargetArray=Get NewMinorMedallions())
   > n1 = GetActiveAugments!(
    execute
    Actor:Actor Object Reference=Get Player State Ref()
    None:%=Get AugmentComponent() -> (AugmentComponent:Fort Player State Component Player Augment System Object Reference)
   ) -> (then, ReturnValue:Array of Fort Player Augment Item Definition Object References)
   > n2 = For Each Loop(Array:Array of Fort Player Augment Item Definition Object References=@n1) -> (Array Element:Fort Player Augment Item Definition Object Reference)
   n3 = GetOwnedGameplayTagBugWorkaround!(self, GameplayTagAssetInterface:Gameplay Tag Asset Interface Interface=@n2) -> (TagContainer:Gameplay Tag Container Structure) {is=title K2Node_CallFunction}
   > n4 = Branch(Condition=HasTag!(
     TagContainer:"Gameplay Tag Container Structure (by ref)"=@n3
     Tag:Gameplay Tag Structure=(TagName="Augment.Category.Medallion")
     bExactMatch:bool=true
    ) -> (ReturnValue:bool))
   > Add(
    TargetArray:Array of Fort Player Augment Item Definition Object References=Get NewBossMedallions()
    NewItem:"Fort Player Augment Item Definition Object Reference (by ref)"=@n2
   ) {is=title K2Node_CallArrayFunction}
   Branch(execute=@n4.else, Condition=HasTag(TagContainer=@n3, Tag=(TagName="Augment.Category.MinorMedallion"), bExactMatch=true))
   > Add(TargetArray=Get NewMinorMedallions(), NewItem=@n2)
   AddNewBossMedallionWidgets(execute=@n2.Completed) {is=title K2Node_CallFunction}
   > AddNewMinorMedallionWidgets() {is=title K2Node_CallFunction}
   > RefreshPopulatedWidgets() {is=title K2Node_CallFunction}
  ]}
]
variables [
 BossMedallionPadding {type=Margin, default=<ltrb 0 0 0 0>}
 BossMedallionSize {type=number, default=0.8}
 CurrentPawn {type=FortPlayerPawn}
 MinorMedallionPadding {type=Margin, default=<ltrb 0 0 5 0>}
 MinorMedallionSize {type=number, default=0.7}
 NewBossMedallions {type="array<FortPlayerAugmentItemDefinition>", default []}
 NewMinorMedallions {type="array<FortPlayerAugmentItemDefinition>", default []}
 Player State Ref {type=FortPlayerStateAthena, default=null}
 TEMP_HorizontalBoxBoonsContainer {type=HorizontalBox, default=null}
 Temp_PopulatedBossMedallions {default {}}
 Temp_PopulatedMinorMedallions {default {}}
]
widgets {
 info {inheritedWidgetCount=1, parentClass=&FortHUDElementWidget, rootWidgetClass=&Overlay, widgetCount=7}
 widgets [
  Overlay_17 {widgetClassPath=&Overlay, widget=Overlay_17}
  ScaleBox_0 {
   widgetClassPath=&ScaleBox
   parent=Overlay_17
   widget=ScaleBox_0
   properties {stretch=UserSpecified, userSpecifiedScale=1.3}
   slot {class=OverlaySlot, properties {verticalAlignment=VAlign_Center}}
  }
  HorizontalBox_0 {widgetClassPath=&HorizontalBox, parent=ScaleBox_0, widget=HorizontalBox_0, slot {class=ScaleBoxSlot}}
  WBP_CherrySmokeBoons_BoonRankBackground {
   widgetClassPath=&/WBP_CherrySmokeBoons_BoonRankBackground
   bIsVariable=true
   parent=HorizontalBox_0
   widget=WBP_CherrySmokeBoons_BoonRankBackground
   properties {elementTag {tagName=Athena.HideForCapture}, hUDElementTag {gameplayTags [{tagName=Athena.HideForCapture}]}}
   slot {class=HorizontalBoxSlot, properties {padding=<ltrb 0 0 8 0>}}
  }
  HorizontalBox_MinorMedallionIcons {
   widgetClassPath=&HorizontalBox
   bIsVariable=true
   parent=HorizontalBox_0
   widget=HorizontalBox_MinorMedallionIcons
   properties {visibility=Collapsed}
   slot {class=HorizontalBoxSlot}
  }
  HorizontalBox_BossMedallionIcons {
   widgetClassPath=&HorizontalBox
   bIsVariable=true
   parent=HorizontalBox_0
   widget=HorizontalBox_BossMedallionIcons
   properties {renderTransform {angle=0, scale=<1.1 1.1>, shear=<0 0>, translation=<0 0>}}
   slot {class=HorizontalBoxSlot, properties {padding=<ltrb 0 -0.5 0 0.5>}}
  }
  TutorialHighlight {widgetClassPath=&TutorialHighlight}
 ]
}
