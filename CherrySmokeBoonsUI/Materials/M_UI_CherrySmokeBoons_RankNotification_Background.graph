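# Translucent UI material that draws the background for the CherrySmokeBoons rank notification as two nested, skewed boxes.
# Antialiased SDF boxes on 9-slice UVs form the shapes, with the inner box smaller and given extra X skew.
# Each box has parameterized multi-stop horizontal color and opacity gradients, with falloff parameters and optional blending toward skewed UVs.
# Named reroutes carry masks, colors and opacities into final lerps that composite the two layers.
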Material /CherrySmokeBoonsUI/Materials/M_UI_CherrySmokeBoons_RankNotification_Background
graph {
 properties {bAutomaticallySetUsageInEditor=false, blendMode=BLEND_Translucent, materialDomain=MD_UI, usedShadingModels=Default Lit}
 nodes [
  LinearInterpolate(
   A=Multiply(A=NamedRerouteUsage(), B=NamedRerouteUsage())
   B=Multiply(A=NamedRerouteUsage(), B=ScalarParameter() {defaultValue=0.9, parameterName=InnerOpacity}) {constB=0.8}
   Alpha=NamedRerouteUsage()
  )
  LinearInterpolate(A=NamedRerouteUsage(), B=NamedRerouteUsage(), Alpha=NamedRerouteUsage())
  NamedRerouteDeclaration(None=AppendVector(A=ScalarParameter() {defaultValue=6.5877, parameterName=SkewX}, B=ScalarParameter() {parameterName=SkewY})) {name=SkewVector, nodeColor=<rgba 1 0 0.9844 1>}
  n1 = NamedRerouteDeclaration(None=MaterialFunctionCall!(UV=MaterialFunctionCall!(UVs) {materialFunction=/FortUILibrary/MaterialFunctions/Basics/MF_UI_9SliceUVs}.UVs, Skew Angle=NamedRerouteUsage()) {materialFunction=/FortUILibrary/MaterialFunctions/Basics/MF_UI_Skew}.Skewed UVs) {name=OuterBoxSkewUVs, nodeColor=<rgba 1 0 0.9141 1>}
  n2 = ScalarParameter() {defaultValue=0.85, parameterName=Size}
  NamedRerouteDeclaration(None=MaterialFunctionCall(DistanceField=MaterialFunctionCall!(UVs=@n1, Size=@n2, CornerRoundness, KeepAspectRatio, PixelSize) {materialFunction=/FortUILibrary/MaterialFunctions/Shapes/SDFs/MF_SDF_Box}.SDF) {materialFunction=&MF_UI_AntiAliasing}.Inner) {name=OuterBox, nodeColor=<rgba 0 1 0.6953 1>}
  n3 = MaterialFunctionCall(Float2=NamedRerouteUsage()) {materialFunction=&BreakOutFloat2Components}
  n4 = Add(A=ScalarParameter() {defaultValue=2.5, parameterName=InternalSkewXOffset}, B=@n3.R)
  n5 = NamedRerouteDeclaration(None=MaterialFunctionCall(
    UV=MaterialFunctionCall() {materialFunction=&MF_UI_9SliceUVs}.UVs
    Skew Angle=AppendVector(A=@n4, B=@n3.G)
   ) {materialFunction=&MF_UI_Skew}.Skewed UVs) {name=InnerBoxSkewUVs, nodeColor=<rgba 1 0.3668 0.1538 1>}
  n6 = Subtract(A=ScalarParameter() {defaultValue=0.85, parameterName=Size}, B=ScalarParameter() {defaultValue=0.1, parameterName=InnerSizeOffset}) {constB=0.1}
  NamedRerouteDeclaration(None=MaterialFunctionCall(DistanceField=MaterialFunctionCall(UVs=@n5, Size=@n6) {materialFunction=&MF_SDF_Box}.SDF) {materialFunction=&MF_UI_AntiAliasing}.Inner) {name=InnerBox, nodeColor=<rgba 1 0.375 0 1>}
  n7 = TextureCoordinate()
  n8 = NamedRerouteUsage()
  n9 = ScalarParameter() {parameterName=SkewColorGradients}
  n10 = Multiply(A=ComponentMask(None=LinearInterpolate(A=@n7, B=@n8, Alpha=@n9)) {r=true}) {constB=3}
  n11 = LinearInterpolate(
   A=VectorParameter() {parameterName=InnerColorA, defaultValue=<rgba 0.0086 0.0823 0.3864 0.502>}.RGB
   B=VectorParameter() {parameterName=InnerColorB, defaultValue=<rgba 0.0273 0.0307 0.2874 1>}.RGB
   Alpha=Saturate(None=@n10)
  )
  n12 = VectorParameter() {parameterName=InnerColorC, defaultValue=<rgba 0 0.1651 0.5029 1>}
  n13 = Subtract(A=@n10)
  n14 = Saturate(None=@n13)
  n15 = Subtract(A=@n13)
  NamedRerouteDeclaration(None=LinearInterpolate(
    A=LinearInterpolate(A=@n11, B=@n12.RGB, Alpha=@n14)
    B=VectorParameter() {parameterName=InnerColorD, defaultValue=<rgba 0.0037 0.1195 0.4397 0.502>}.RGB
    Alpha=Saturate(None=@n15)
   )) {name=InnerBoxColor, nodeColor=<rgba 1 0 0.5859 1>}
  n16 = VectorParameter() {parameterName=OuterColorA, defaultValue=<rgba 0 0.1946 0.8796 0.502>}
  n17 = VectorParameter() {parameterName=OuterColorB, defaultValue=<rgba 0.162 0.5841 0.5841 1>}
  n18 = TextureCoordinate()
  n19 = NamedRerouteUsage()
  n20 = ScalarParameter() {parameterName=SkewColorGradients}
  n21 = Multiply(A=ComponentMask(None=LinearInterpolate(A=@n18, B=@n19, Alpha=@n20)) {r=true}) {constB=2}
  n22 = Saturate(None=@n21)
  n23 = Subtract(A=@n21)
  NamedRerouteDeclaration(None=LinearInterpolate(
    A=LinearInterpolate(A=@n16.RGB, B=@n17.RGB, Alpha=@n22)
    B=VectorParameter() {parameterName=OuterColorC, defaultValue=<rgba 0 0.3278 0.6654 0.502>}.RGB
    Alpha=Saturate(None=@n23)
   )) {name=OuterBoxColor, nodeColor=<rgba 0 0.9062 1 1>}
  n24 = ScalarParameter() {defaultValue=0.1, parameterName=InnerOpacityFalloff}
  n25 = TextureCoordinate()
  n26 = NamedRerouteUsage()
  n27 = ScalarParameter() {parameterName=SkewOpacityGradients}
  n28 = Multiply(A=ComponentMask(None=LinearInterpolate(A=@n25, B=@n26, Alpha=@n27)) {r=true}) {constB=3}
  n29 = LinearInterpolate(A=Multiply(A=ScalarParameter() {defaultValue=0.9, parameterName=InnerOpacity}, B=OneMinus(None=@n24)), Alpha=Saturate(None=@n28)) {constA=0.5}
  n30 = Subtract(A=@n28)
  n31 = Saturate(None=@n30)
  n32 = ScalarParameter() {defaultValue=0.9, parameterName=InnerOpacity}
  n33 = OneMinus(None=ScalarParameter() {defaultValue=0.1, parameterName=InnerOpacityFalloff})
  n34 = Subtract(A=@n30)
  NamedRerouteDeclaration(None=LinearInterpolate(A=LinearInterpolate(A=@n29, Alpha=@n31), B=Multiply(A=@n32, B=@n33), Alpha=Saturate(None=@n34)) {constB=0.5}) {name=InnerOpacity, nodeColor=<rgba 1 0 0.75 1>}
  n35 = ScalarParameter() {defaultValue=0.2, parameterName=OuterOpacityFalloff}
  n36 = TextureCoordinate()
  n37 = NamedRerouteUsage()
  n38 = ScalarParameter() {parameterName=SkewOpacityGradients}
  n39 = Multiply(A=ComponentMask(None=LinearInterpolate(A=@n36, B=@n37, Alpha=@n38)) {r=true}) {constB=3}
  n40 = LinearInterpolate(A=Multiply(A=ScalarParameter() {defaultValue=1, parameterName=OuterOpacity}, B=OneMinus(None=@n35)), Alpha=Saturate(None=@n39)) {constA=0.5}
  n41 = Subtract(A=@n39)
  n42 = Saturate(None=@n41)
  n43 = ScalarParameter() {defaultValue=1, parameterName=OuterOpacity}
  n44 = OneMinus(None=ScalarParameter() {defaultValue=0.2, parameterName=OuterOpacityFalloff})
  n45 = Subtract(A=@n41)
  NamedRerouteDeclaration(None=LinearInterpolate(A=LinearInterpolate(A=@n40, Alpha=@n42), B=Multiply(A=@n43, B=@n44), Alpha=Saturate(None=@n45)) {constB=0.5}) {name=OuterOpacity, nodeColor=<rgba 0 1 0.9297 1>}
 ]
}
