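# HUD widget that shows a "RANKED UP!" notification with rank badge and boon reward text in CherrySmokeBoons.
# HandleShowRankUpNotification skips the display for C rank, otherwise picks the S+ or normal message and plays a fade-and-slide animation.
# A timer, longer for S+, reverses the animation, then a second timer collapses the widget.
# UpdateRankIcon sets the badge from a soft material and zeroes its pip parameters.
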WidgetBlueprint /CherrySmokeBoonsUI/HUD/WBP_CherrySmokeBoons_RankUpNotification
animations [Anim_Show {
  binding_count=1
  length_seconds=0.5
  object=&self:Anim_Show
  sequence {bindings [Overlay_RankUp_Message {
     id=#1
     possessed_class=Overlay
     tracks [
      {
       type=MovieSceneFloatTrack
       display_name=Render Opacity
       sections [{type=MovieSceneFloatSection, end_frame=10, start_frame=0, channels [None {type=MovieSceneScriptingFloatChannel, keys [[0,0],[10,1]]}]}]
      }
      {
       type=MovieScene2DTransformTrack
       sections [{
         type=MovieScene2DTransformSection
         end_frame=10
         start_frame=0
         channels [
          Translation.X {type=MovieSceneScriptingFloatChannel, const [0,2]}
          Translation.Y {type=MovieSceneScriptingFloatChannel, keys [[0,40],[10,0]]}
          Angle {type=MovieSceneScriptingFloatChannel, const [0,1]}
          Scale.X {type=MovieSceneScriptingFloatChannel, const [1,1]}
          Scale.Y {type=MovieSceneScriptingFloatChannel, const [1,1]}
          Shear.X {type=MovieSceneScriptingFloatChannel, const [0,1]}
          Shear.Y {type=MovieSceneScriptingFloatChannel, const [0,1]}
         ]
        }]
      }
     ]
    }]}
 }]
graphs [
 EventGraph {nodes [
   n1 = OnShowRankUpNotification!() -> (OutputDelegate, then, RankTag:Gameplay Tag Structure) {is=title K2Node_CustomEvent}
   > n2 = HandleShowRankUpNotification!(execute, self, RankUpTag:Gameplay Tag Structure=@n1.RankTag) -> (then, IsSPlusRank:bool)
   n3 = OnShowEnd() {is=title K2Node_CustomEvent}
   n4 = Set Timer by Event(
    execute=@n2.then
    Delegate=@n3
    Time=SelectFloat(A=Get SPlusDuration() -> (SPlusDuration:double), B=Get NormalDuration() -> (NormalDuration:double), bPickA=@n2)
    bLooping=false
   )
   > Set ShowEndTimerHandle(ShowEndTimerHandle:Timer Handle Structure=@n4) -> (Output_Get:Timer Handle Structure)
   n5 = UpdateRankIcon!() -> (OutputDelegate, then, NewRankIcon:Material Instance Soft Object Reference) {is=title K2Node_CustomEvent}
   n6 = Get Image_RankIcon() -> (Image_RankIcon:Image Object Reference)
   > SetBrushFromSoftMaterial!(execute, self:Image Object Reference=@n6, SoftMaterial:Material Interface Soft Object Reference=@n5.NewRankIcon) -> (then)
   > n7 = GetDynamicMaterial(self=@n6)
   > SetScalarParameterValue(self=@n7, ParameterName=ShowMiddlePips, Value=0)
   > SetScalarParameterValue(self=@n7, ParameterName=BackgroundPipsOpacity, Value=0)
   n8 = OnHideEnd() {is=title K2Node_CustomEvent}
   > SetVisibility(InVisibility=Collapsed)
   > Clear and Invalidate Timer by Handle(Handle=Get HideEndTimerHandle() -> (HideEndTimerHandle:Timer Handle Structure))
   Clear and Invalidate Timer by Handle(execute=@n10.then, Handle=Get ShowEndTimerHandle() -> (ShowEndTimerHandle:Timer Handle Structure))
   Event Construct()
   PlayAnimationReverse(execute=@n3.then, InAnimation=Get Anim_Show() -> (Anim_Show:Widget Animation Object Reference))
   > n9 = Set Timer by Event(Delegate=@n8, Time=0.5, bLooping=false)
   > n10 = Set HideEndTimerHandle(HideEndTimerHandle:Timer Handle Structure=@n9) -> (Output_Get:Timer Handle Structure)
  ]}
 HandleShowRankUpNotification {nodes [
   n1 = HandleShowRankUpNotification!() -> (then, RankUpTag:Gameplay Tag Structure) {is=title K2Node_FunctionEntry}
   > n2 = Branch(Condition="Equal (GameplayTag)"(A=@n1, B=Get CRankTag() -> (CRankTag:Gameplay Tag Structure)))
   n3 = "Equal (GameplayTag)"(A=Get RankUpTag() -> (RankUpTag:Gameplay Tag Structure), B=Get SPlusRankTag() -> (SPlusRankTag:Gameplay Tag Structure))
   Set Text(
    execute=@n2.else
    Text=Select(
     Option 0:Text=Get Text_BoonsEnhanced() -> (Text_BoonsEnhanced:Text)
     Option 1:Text=Get Text_AllBoonsAwarded() -> (Text_AllBoonsAwarded:Text)
     Index=@n3
    ) -> (ReturnValue:Text)
    self=Get Text_Description() -> (Text_Description:Text Block Object Reference)
   )
   > SetVisibility(InVisibility=HitTestInvisible)
   > PlayAnimationForward(InAnimation=Get Anim_Show())
   > Return Node!(execute, IsSPlusRank:bool=@n3)
  ]}
]
variables [
 HideEndTimerHandle {type=TimerHandle}
 NormalDuration {type=number, default=6}
 ShowEndTimerHandle {type=TimerHandle}
 SPlusDuration {type=number, default=8}
 Text_AllBoonsAwarded {type=string, default="All Boons Awarded - Health && Shield Gained"}
 Text_BoonsEnhanced {type=string, default="Boons Enhanced - Health && Shield Gained"}
]
widgets {
 info {inheritedWidgetCount=1, parentClass=&FortHUDElementWidget, rootWidgetClass=&Overlay, widgetCount=7}
 widgets [
  Overlay_RankUp_Message {widgetClassPath=&Overlay, bIsVariable=true, widget=Overlay_RankUp_Message}
  Border_Background {
   widgetClassPath=&Border
   bIsVariable=true
   parent=Overlay_RankUp_Message
   widget=Border_Background
   properties {
    horizontalAlignment=HAlign_Left
    verticalAlignment=VAlign_Top
    background {
     bIsDynamicallyLoaded=false
     drawAs=Image
     imageType=FullColor
     mirroring=NoMirror
     resourceName=None
     resourceObject=/CherrySmokeBoonsUI/Materials/M_UI_CherrySmokeBoons_RankNotification_Background
     tiling=NoTile
     imageSize=<32 32>
     margin=<ltrb 0 0 0 0>
     outlineSettings {
      bUseBrushTransparency=false
      roundingType=HalfHeightRadius
      width=0
      color {colorUseRule=UseColor_Specified, specifiedColor=<rgba 0 0 0 0>}
      cornerRadii=<0 0 0 0>
     }
     tintColor {colorUseRule=UseColor_Specified, specifiedColor=<rgba 1 1 1 1>}
     uVRegion {bIsValid=false, max=<0 0>, min=<0 0>}
    }
    padding=<ltrb 30 25 30 25>
   }
   slot {class=OverlaySlot}
  }
  GridPanel_31 {
   widgetClassPath=&GridPanel
   parent=Border_Background
   widget=GridPanel_31
   slot {class=BorderSlot, properties {horizontalAlignment=HAlign_Left, verticalAlignment=VAlign_Top, padding=<ltrb 30 25 30 25>}}
  }
  Image_RankIcon {
   widgetClassPath=&Image
   bIsVariable=true
   parent=GridPanel_31
   widget=Image_RankIcon
   properties {
    brush {
     bIsDynamicallyLoaded=false
     drawAs=Image
     imageType=FullColor
     mirroring=NoMirror
     resourceName=None
     resourceObject=/CherrySmokeHeroRankingUI/Materials/BadgeInstanceMaterials/MI_UI_CherrySmokeHeroRank_FullBadge_C
     tiling=NoTile
     imageSize=<64 64>
     margin=<ltrb 0 0 0 0>
     outlineSettings {
      bUseBrushTransparency=false
      roundingType=HalfHeightRadius
      width=0
      color {colorUseRule=UseColor_Specified, specifiedColor=<rgba 0 0 0 0>}
      cornerRadii=<0 0 0 0>
     }
     tintColor {colorUseRule=UseColor_Specified, specifiedColor=<rgba 1 1 1 1>}
     uVRegion {bIsValid=false, max=<0 0>, min=<0 0>}
    }
    renderTransform {angle=0, scale=<1.5 1.5>, shear=<0 0>, translation=<0 0>}
   }
   slot {class=GridSlot, properties {horizontalAlignment=HAlign_Left, rowSpan=2, verticalAlignment=VAlign_Center, padding=<ltrb 0 0 15 0>}}
  }
  Text_RankedUp {
   widgetClassPath=&UEFN_TextBlock
   bIsVariable=true
   parent=GridPanel_31
   widget=Text_RankedUp
   properties {
    text="RANKED UP!"
    font {
     bForceMonospaced=false
     bMaterialIsStencil=false
     fontMaterial=null
     fontObject=&BurbankSmall
     letterSpacing=0
     monospacedWidth=1
     size=18
     skewAmount=0
     typefaceFontName=Black
     outlineSettings {
      bApplyOutlineToDropShadows=false
      bMiteredCorners=false
      bSeparateFillAlpha=false
      outlineMaterial=null
      outlineSize=0
      outlineColor=<rgba 0 0 0 1>
     }
    }
   }
   slot {class=GridSlot, properties {column=1, horizontalAlignment=HAlign_Left, verticalAlignment=VAlign_Center}}
  }
  Text_Description {
   widgetClassPath=&UEFN_TextBlock
   bIsVariable=true
   parent=GridPanel_31
   widget=Text_Description
   properties {
    text="Boons Enhanced - Health && Shield Gained"
    visibility=HitTestInvisible
    wrapTextAt=575
    font {
     bForceMonospaced=false
     bMaterialIsStencil=false
     fontMaterial=null
     fontObject=&BurbankSmall
     letterSpacing=0
     monospacedWidth=1
     size=16.5
     skewAmount=0
     typefaceFontName=Black
     outlineSettings {
      bApplyOutlineToDropShadows=false
      bMiteredCorners=false
      bSeparateFillAlpha=false
      outlineMaterial=null
      outlineSize=0
      outlineColor=<rgba 0 0 0 1>
     }
    }
   }
   slot {class=GridSlot, properties {column=1, horizontalAlignment=HAlign_Left, row=1, verticalAlignment=VAlign_Top}}
  }
  TutorialHighlight {widgetClassPath=&TutorialHighlight}
 ]
}
